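# Unlit Modulate Niagara material for the Hydraulics emote that fakes four soft pools of projected light (left, right, front, rear) on scene surfaces.
# It transforms world position behind translucency to local space, then scales and offsets it per side to build power-sharpened sphere masks.
# The summed masks are tinted by particle color and depth-faded.
# A minimum clamp of 1 keeps modulation from darkening, before MF_FX_EmissiveAdjustments processes the result.
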Material /BRCosmetics/Effects/Fort_Effects/Emotes/Hydraulics/M_LightProject
graph {
 properties {
  bUsedWithNiagaraMeshParticles=true
  bUsedWithNiagaraSprites=true
  blendMode=BLEND_Modulate
  pixelDepthOffsetMode=PDOM_Legacy
  shadingModel=MSM_Unlit
  translucencyPass=MTP_BeforeDOF
  usedShadingModels=Unlit
 }
 nodes [
  n1 = TransformPosition(None=MaterialFunctionCall() {materialFunction=&WorldPositionBehindTranslucency}.Result) {transformSourceType=TRANSFORMPOSSOURCE_World}
  n2 = Divide(A=@n1, B=VectorParameter() {group=World Position, parameterName=Sides Divide, defaultValue=<rgba 100 200 0 1>}.RGB) {constB=50}
  n3 = VectorParameter() {group=Left Radius, parameterName=Left Radius Location Offset, defaultValue=<rgba 0.15 0.5 0 1>}
  n4 = Saturate(None=ComponentMask(None=Add(A=@n2, B=@n3.RGB) {constB=0.5}) {g=true, r=true})
  n5 = Constant2Vector() {g=0.5, r=0.5}
  n6 = ScalarParameter() {defaultValue=1.2414, group=Left Radius, parameterName=Left Radius}
  n7 = VectorParameter() {group=Right Radius, parameterName=Right Radius Location Offset, defaultValue=<rgba 0.85 0.5 0 1>}
  n8 = Saturate(None=ComponentMask(None=Add(A=@n2, B=@n7.RGB) {constB=0.5}) {g=true, r=true})
  n9 = ScalarParameter() {defaultValue=1.2414, group=Right Radius, parameterName=Right Radius}
  n10 = Add(
   A=Power(
    Base=SphereMask(A=@n4, B=@n5, Radius=@n6) {attenuationRadius=0.5, hardnessPercent=0}
    Exp=ScalarParameter() {defaultValue=6.7003, group=Left Radius, parameterName=Left Power}
   )
   B=Power(
    Base=SphereMask(A=@n8, B=@n5, Radius=@n9) {attenuationRadius=0.5, hardnessPercent=0}
    Exp=ScalarParameter() {defaultValue=6.7003, group=Right Radius, parameterName=Right Power}
   )
  )
  n11 = Divide(A=@n1, B=VectorParameter() {group=World Position, parameterName=Front Back Divide, defaultValue=<rgba 200 100 0 1>}.RGB) {constB=50}
  n12 = ComponentMask(None=Add(A=@n11, B=VectorParameter() {group=Front Radius, parameterName=Front Radius Location Offset, defaultValue=<rgba 0.5 -0.2 0 1>}.RGB) {constB=0.5}) {g=true, r=true}
  n13 = Power(
   Base=SphereMask(A=Saturate(None=@n12), B=@n5, Radius=ScalarParameter() {defaultValue=0.9096, group=Front Radius, parameterName=Front Radius}) {attenuationRadius=0.5, hardnessPercent=0}
   Exp=ScalarParameter() {defaultValue=6.7003, group=Front Radius, parameterName=Front Power}
  )
  n14 = Add(A=@n11, B=VectorParameter() {group=Rear Radius, parameterName=Rear Radius Location Offset, defaultValue=<rgba 0.5 1.1 0 1>}.RGB) {constB=0.5}
  n15 = SphereMask(
   A=Saturate(None=ComponentMask(None=@n14) {g=true, r=true})
   B=@n5
   Radius=ScalarParameter() {defaultValue=0.9096, group=Rear Radius, parameterName=Rear Radius}
  ) {attenuationRadius=0.5, hardnessPercent=0}
  n16 = ScalarParameter() {defaultValue=6.7003, group=Rear Radius, parameterName=Rear Power}
  n17 = Multiply(A=ParticleColor().RGB, B=Add(A=Add(A=@n10, B=@n13), B=Power(Base=@n15, Exp=@n16)))
  n18 = ScalarParameter() {defaultValue=32, group=Depth Fade, parameterName=FadeDistance}
  MaterialFunctionCall(Emissive=Clamp(None=DepthFade(Opacity=@n17, FadeDistance=@n18) {fadeDistanceDefault=32}) {maxDefault=100, minDefault=1}) {materialFunction=&MF_FX_EmissiveAdjustments}
 ]
}
