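# Niagara system for the Hydraulics emote that adds a persistent green light with a companion light decal.
# The Light emitter bursts one particle just above the system location, drawn by both light and component renderers.
# LightDecal bursts one brighter green particle closer to the origin, drawn by a mesh renderer with fixed sprite facing and alignment.
# Both particles loop their lifetime instead of dying, and no shown module reads the boost, speed or scale user parameters.
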NiagaraSystem /BRCosmetics/Effects/Fort_Effects/Emotes/Hydraulics/NS_Emote_Hydraulics_Light
dependencies {usedDataInterfaces [], usedDynamicInputs [], usedModules [/Niagara/Modules/System/SystemLifeCycle, &SystemState]}
emitters {
 Light {
  emitterUpdateScript [
   &EmitterState(Life Cycle Mode=::NewEnumerator1, Loop Behavior=::NewEnumerator1, UseLoopDelay=false)
   &SpawnBurst_Instantaneous(Spawn Count=1, Use Spawn Probability=false)
  ]
  particleSpawnScript [
   &InitializeParticle(Color=<rgba 0 1 0 1>, Lifetime=5, UsePositionOffset=false, Write Material Random=true)
   &SystemLocation(MaskSpawnGroup=false, Offset=<0 0 10>)
  ]
  particleUpdateScript [&ParticleState(Kill Particles When Lifetime Has Elapsed=false, Loop Particles Lifetime=true)]
  rendererClasses [&NiagaraLightRendererProperties, &NiagaraComponentRendererProperties]
  renderers [
   {rendererClass=&NiagaraLightRendererProperties, rendererIndex=0}
   {rendererClass=&NiagaraComponentRendererProperties, rendererIndex=1}
  ]
 }
 LightDecal {
  emitterUpdateScript [
   &EmitterState(Life Cycle Mode=::NewEnumerator1, Loop Behavior=::NewEnumerator1, UseLoopDelay=false)
   &SpawnBurst_Instantaneous(Spawn Count=1, Use Spawn Probability=false)
  ]
  particleSpawnScript [
   &InitializeParticle(
    Color=<rgba 0 5 0 1>
    Lifetime=5
    Sprite Size=<200 250>
    Sprite Size Mode=::NewEnumerator4
    UsePositionOffset=false
    Write Material Random=true
   )
   &SystemLocation(MaskSpawnGroup=false, Offset=<0 0 2>)
  ]
  particleUpdateScript [
   &ParticleState(Kill Particles When Lifetime Has Elapsed=false, Loop Particles Lifetime=true)
   /SetVariables(Particles.SpriteFacing=<0 0 90>) {isSetParameters=true}
   /SetVariables(Particles.SpriteAlignment=<0 1 0>) {isSetParameters=true}
  ]
  rendererClasses [&NiagaraMeshRendererProperties]
  renderers [{rendererClass=&NiagaraMeshRendererProperties, rendererIndex=0}]
 }
}
summary {
 systemName=NS_Emote_Hydraulics_Light
 emitters [
  {emitterName=Light, rendererClasses [&NiagaraLightRendererProperties, &NiagaraComponentRendererProperties]}
  {emitterName=LightDecal, rendererClasses [&NiagaraMeshRendererProperties]}
 ]
 userVariables [
  User.EffectScale {type=NiagaraFloat, defaultValue=1, description=""}
  User.EffectSpawnScale {type=NiagaraFloat, defaultValue=10, description=""}
  User.ForwardSpeed {type=NiagaraFloat, defaultValue=0, description=""}
  User.isBoosting {type=NiagaraBool}
 ]
}
